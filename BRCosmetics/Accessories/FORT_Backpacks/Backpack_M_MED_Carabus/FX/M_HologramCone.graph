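# Additive, two-sided material for a hologram cone effect on the Carabus backpack, usable on Niagara mesh particles.
# Particle color is multiplied by a steep vertical UV gradient and a sharpened Fresnel term, then passed through MF_FX_EmissiveAdjustments.
# A separate quality switch yields a constant one, except on Low, where it multiplies a Fresnel by the linear UV gradient.
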Material /BRCosmetics/Accessories/FORT_Backpacks/Backpack_M_MED_Carabus/FX/M_HologramCone
graph {
 properties {bUsedWithNiagaraMeshParticles=true, blendMode=BLEND_Additive, pixelDepthOffsetMode=PDOM_Legacy, twoSided=true, usedShadingModels=Default Lit}
 nodes [
  n1 = ParticleColor()
  n2 = ComponentMask(None=TextureCoordinate()) {g=true}
  n3 = Power(Base=@n2, Exp=Constant() {r=4})
  n4 = Fresnel(ExponentIn=Constant() {r=3})
  n5 = Constant() {r=1.5}
  MaterialFunctionCall(Emissive=Multiply(A=Multiply(A=@n1.RGB, B=@n3), B=Power(Base=@n4, Exp=@n5))) {materialFunction=&MF_FX_EmissiveAdjustments}
  n6 = Constant() {r=1}
  n7 = Constant() {r=2}
  QualitySwitch(Default=@n6, Low=Multiply(A=Fresnel(ExponentIn=@n7), B=@n2), High=@n6, Medium=@n6, Epic=@n6)
 ]
}
