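# Translucent Niagara sprite material for wisp effects on the Carabus backpack cosmetic.
# Four chained Perlin-curl samples distort UVs using dynamic parameters, then two panning noise channels are combined with Max into wispy shapes.
# A soft radial mask, particle alpha and depth fade shape that opacity.
# Particle color brightens along a transformed Z gradient, then gets an opacity-driven HSV shift, with HSV value also scaling the faded opacity.
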Material /BRCosmetics/Accessories/FORT_Backpacks/Backpack_M_MED_Carabus/FX/M_WispSprites
graph {
 properties {bUsedWithNiagaraSprites=true, blendMode=BLEND_Translucent, pixelDepthOffsetMode=PDOM_Legacy, usedShadingModels=Default Lit}
 nodes [
  Constant3Vector() {constant=<rgba 0.125 0.125 0.125 0>}
  Constant() {r=0.9}
  n1 = ParticleColor()
  n2 = TextureCoordinate() {uTiling=0.2, vTiling=0.2}
  n3 = TextureObject() {samplerType=SAMPLERTYPE_LinearColor, texture=/Game/Accessories/FORT_Backpacks/Backpack_M_MED_Carabus/FX/T_PerlinCurl}
  n4 = ComponentMask(None=TextureSample(UVs=@n2, Tex=@n3) {samplerType=SAMPLERTYPE_LinearColor}.RGB) {g=true, r=true}
  n5 = Constant2Vector() {g=0.5, r=0.5}
  n6 = DynamicParameter()
  n7 = Multiply(A=Constant() {r=0.0625}, B=LinearInterpolate(Alpha=@n6.G) {constA=0.25, constB=-0.25})
  n8 = Add(A=@n2, B=Multiply(A=Subtract(A=@n4, B=@n5), B=@n7))
  n9 = ComponentMask(None=TextureSample(UVs=@n8, Tex=@n3) {samplerType=SAMPLERTYPE_LinearColor}.RGB) {g=true, r=true}
  n10 = Constant2Vector() {g=0.5, r=0.5}
  n11 = Add(A=@n8, B=Multiply(A=Subtract(A=@n9, B=@n10), B=@n7))
  n12 = ComponentMask(None=TextureSample(UVs=@n11, Tex=@n3) {samplerType=SAMPLERTYPE_LinearColor}.RGB) {g=true, r=true}
  n13 = Constant2Vector() {g=0.5, r=0.5}
  n14 = Add(A=@n11, B=Multiply(A=Subtract(A=@n12, B=@n13), B=@n7))
  n15 = TextureSample(UVs=@n14, Tex=@n3) {samplerType=SAMPLERTYPE_LinearColor}
  n16 = Multiply(A=Subtract(A=ComponentMask(None=@n15.RGB) {g=true, r=true}, B=Constant2Vector() {g=0.5, r=0.5}), B=@n7)
  n17 = Add(A=Multiply(A=Add(A=@n14, B=@n16), B=ScalarParameter() {defaultValue=2.5}), B=@n6.R)
  n18 = Max(
   A=TextureSample(UVs=Panner(Coordinate=@n17) {speedY=-0.1}) {texture=&T_NoisePack_01}.R
   B=TextureSample(UVs=Panner(Coordinate=@n17) {speedY=-0.2}) {texture=&T_NoisePack_01}.G
  )
  n19 = SphereMask(A=TextureCoordinate(), B=Constant2Vector() {g=0.5, r=0.5}) {attenuationRadius=0.5, hardnessPercent=0}
  n20 = Multiply(A=@n1.A, B=Multiply(A=Power(Base=@n18) {constExponent=2.4}, B=Multiply(A=@n19, B=@n19)))
  n21 = DepthFade(Opacity=Multiply(A=Saturate(None=@n20), B=Constant() {r=0.3})) {fadeDistanceDefault=16}
  n22 = AppendVector(
   A=MaterialFunctionCall(Input=@n21, Input Low=Constant(), Input High=Constant() {r=1}, Target Low=Constant() {r=0.95}, Target High=Constant() {r=1}) {materialFunction=&RemapValueRange}.Result
   B=Constant() {r=-0.1}
  )
  n23 = Constant() {r=-0.1}
  n24 = WorldPosition()
  n25 = Subtract(A=ComponentMask(None=TransformPosition(None=@n24.XYZ) {transformSourceType=TRANSFORMPOSSOURCE_World}) {b=true}, B=Constant() {r=-70})
  n26 = Constant() {r=40}
  n27 = LinearInterpolate(A=Multiply(A=@n1.RGB, B=Constant() {r=3}), B=@n1.RGB, Alpha=Saturate(None=Divide(A=@n25, B=@n26)))
  n28 = MaterialFunctionCall!(RGB=@n27) {materialFunction=/Engine/Functions/Engine_MaterialFunctions02/Math/RGBtoHSV}
  MaterialFunctionCall!(HSV_In=Add(A=AppendVector(A=@n22, B=@n23), B=@n28."RGB to HSV (0-1)")) {materialFunction=/Game/Athena/UI/Materials/MF_HSVtoRGB}
  Multiply(
   A=MaterialFunctionCall(
    Input=MaterialFunctionCall(Float3=@n28."RGB to HSV (0-1)") {materialFunction=&BreakOutFloat3Components}.B
    Input Low=Constant() {r=0.3}
    Input High=Constant() {r=1}
    Target Low=Constant() {r=0.3}
    Target High=Constant() {r=0.2}
   ) {materialFunction=&RemapValueRange}.Result
   B=@n21
  )
  OneMinus(None=@n27)
  Saturate(None=@n27)
 ]
}
