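# Physics asset for the Carabus backpack cosmetic (M_MED), defining capsule bodies for its two dangling belts and the wearer's lower body and spine.
# Each belt is an eleven-link chain hung from the pack, with tight swing limits and twist limits that narrow down the chain, tuned differently per side.
# Pelvis, spine, thigh and calf capsules give body volume, each tied directly to body_root by fully free angular constraints.
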PhysicsAsset /BRCosmetics/Accessories/FORT_Backpacks/Backpack_M_MED_Carabus/Meshes/M_MED_Carabus_Pack_Physics
bodies {
 body_root [{shapeType=Capsule, length=0.505, radius=0.505, shapeName=body_root_capsule, rotation=<rot 0 -90 0>}]
 calf_l [{shapeType=Capsule, length=33.3955, radius=8.4272, center=<-16.1828 0.0001 0.4414>, rotation=<rot -88.4247 179.9991 -179.9991>}]
 calf_r [{shapeType=Capsule, length=33.0334, radius=8.4272, center=<17.3473 0.0001 0>, rotation=<rot 90 0.0198 0.0198>}]
 dyn_belt_l_01 [{shapeType=Capsule, length=4.505, radius=1.255, center=<-0.0171 0.0329 -1.9333>, rotation=<rot 0 90.6424 -179.7933>}]
 dyn_belt_l_010 [{shapeType=Capsule, length=7.9892, radius=3.9796, center=<0.0777 -0.0835 -5.9022>, rotation=<rot 0.1178 -90 0>}]
 dyn_belt_l_011 [{shapeType=Capsule, length=10.5129, radius=3.9796, center=<0.0778 -0.5468 -7.0786>, rotation=<rot -0.0919 -90 0>}]
 dyn_belt_l_02 [{shapeType=Capsule, length=3.8622, radius=1.2823, center=<0.0113 0.1232 -2.2137>, rotation=<rot -1.8665 -90 -0.381>}]
 dyn_belt_l_03 [{shapeType=Capsule, length=3.2743, radius=1.2812, center=<0.0523 0.0391 -1.9868>, rotation=<rot -0.1148 -90 -0.499>}]
 dyn_belt_l_04 [{shapeType=Capsule, length=5.2055, radius=3.152, center=<0.2822 -0.2051 -2.786>, rotation=<rot -1.4003 -90 3.6946>}]
 dyn_belt_l_05 [{shapeType=Capsule, length=4.7837, radius=3.1783, center=<0.0982 0.0657 -3.7902>, rotation=<rot -2.0934 -90 -0.1381>}]
 dyn_belt_l_06 [{shapeType=Capsule, length=4.141, radius=3.1951, center=<0.2134 0.3757 -3.9327>, rotation=<rot -1.1671 -90 0.0552>}]
 dyn_belt_l_07 [{shapeType=Capsule, length=6.1642, radius=3.2634, center=<0.0814 0.0649 -4.5694>, rotation=<rot 2.5694 -89.5793 0.8813>}]
 dyn_belt_l_08 [{shapeType=Capsule, length=6.458, radius=3.9796, center=<0.0778 -0.2633 -5.0888>, rotation=<rot -0.0175 -90 0>}]
 dyn_belt_l_09 [{shapeType=Capsule, length=6.8559, radius=3.9796, center=<0.0779 -0.1035 -5.2756>, rotation=<rot -0.3048 -90 0>}]
 dyn_belt_r_01 [{shapeType=Capsule, length=4.755, radius=1.255, center=<-0.1106 -0.0359 1.9465>, rotation=<rot 0 90.9726 179.9321>}]
 dyn_belt_r_010 [{shapeType=Capsule, length=9.4526, radius=3.9796, center=<-0.0577 -0.1066 5.2919>, rotation=<rot 0.0036 -90 0>}]
 dyn_belt_r_011 [{shapeType=Capsule, length=4.7322, radius=3.9796, center=<-0.0578 0.6234 4.2389>, rotation=<rot -0.0008 -90 0>}]
 dyn_belt_r_02 [{shapeType=Capsule, length=3.8622, radius=1.2847, center=<-0.1152 -0.1296 2.2201>, rotation=<rot -1.896 -90 -0.1946>}]
 dyn_belt_r_03 [{shapeType=Capsule, length=3.2742, radius=1.2834, center=<-0.1439 -0.0494 1.9913>, rotation=<rot -0.1457 -90 -0.4104>}]
 dyn_belt_r_04 [{shapeType=Capsule, length=4.7055, radius=3.0828, center=<-0.2359 0.2771 3.3828>, rotation=<rot -3.4238 -90 3.5261>}]
 dyn_belt_r_05 [{shapeType=Capsule, length=5.2837, radius=3.1193, center=<-0.0973 -0.0678 3.844>, rotation=<rot -2.0936 -90 -0.1346>}]
 dyn_belt_r_06 [{shapeType=Capsule, length=3.641, radius=3.1237, center=<-0.2055 -0.3272 3.9611>, rotation=<rot -3.8713 -89.9999 1.8893>}]
 dyn_belt_r_07 [{shapeType=Capsule, length=6.1437, radius=3.1883, center=<-0.248 -0.5445 4.0553>, rotation=<rot 1.4842 -89.7864 1.9894>}]
 dyn_belt_r_08 [{shapeType=Capsule, length=6.5733, radius=3.9796, center=<-0.0577 0.1178 5.1533>, rotation=<rot -0.0513 -90 0>}]
 dyn_belt_r_09 [{shapeType=Capsule, length=7.0892, radius=3.9796, center=<-0.0576 0.2107 4.3905>, rotation=<rot 0.1498 -90 0>}]
 pack [{shapeType=Capsule, length=7.4815, radius=9.9465, center=<-0.1976 -6.1625 -1.8932>, rotation=<rot -9.8083 -90.0015 0.031>}]
 pelvis [{shapeType=Capsule, length=5.5, radius=15, center=<-4.857 0 0>}]
 spine_01 [{shapeType=Capsule, length=13.2006, radius=13.25}]
 spine_02 [{shapeType=Capsule, length=14.7006, radius=11.75}]
 thigh_l [{shapeType=Capsule, length=32.279, radius=9, center=<-23.2064 0.0001 0.3354>, rotation=<rot -89.172 179.999 -179.999>}]
 thigh_r [{shapeType=Capsule, length=31.4177, radius=9, center=<22.7898 0.0002 -0.3292>, rotation=<rot -89.172 179.999 -179.999>}]
}
constraints [
 {bone1Name=dyn_belt_l_01, bone2Name=pack, swing1LimitDegrees=2.5, swing2LimitDegrees=6, twistLimitDegrees=40}
 {bone1Name=dyn_belt_l_02, bone2Name=dyn_belt_l_01, swing1LimitDegrees=2.5, swing2LimitDegrees=6, twistLimitDegrees=20}
 {bone1Name=dyn_belt_l_03, bone2Name=dyn_belt_l_02, swing1LimitDegrees=2.5, swing2LimitDegrees=6, twistLimitDegrees=10}
 {bone1Name=dyn_belt_r_01, bone2Name=pack, swing1LimitDegrees=5, swing2LimitDegrees=5, twistLimitDegrees=40}
 {bone1Name=dyn_belt_r_02, bone2Name=dyn_belt_r_01, swing1LimitDegrees=5, swing2LimitDegrees=5, twistLimitDegrees=35}
 {bone1Name=dyn_belt_r_03, bone2Name=dyn_belt_r_02, swing1LimitDegrees=5, swing2LimitDegrees=5, twistLimitDegrees=30}
 {bone1Name=dyn_belt_r_07, bone2Name=dyn_belt_r_06, swing1LimitDegrees=5, swing2LimitDegrees=5, twistLimitDegrees=20}
 {bone1Name=dyn_belt_r_06, bone2Name=dyn_belt_r_05, swing1LimitDegrees=5, swing2LimitDegrees=5, twistLimitDegrees=20}
 {bone1Name=dyn_belt_r_04, bone2Name=dyn_belt_r_03, swing1LimitDegrees=5, swing2LimitDegrees=5, twistLimitDegrees=25}
 {bone1Name=dyn_belt_r_05, bone2Name=dyn_belt_r_04, swing1LimitDegrees=5, swing2LimitDegrees=5, twistLimitDegrees=20}
 {bone1Name=dyn_belt_l_07, bone2Name=dyn_belt_l_06, swing1LimitDegrees=2.5, swing2LimitDegrees=6, twistLimitDegrees=10}
 {bone1Name=dyn_belt_l_06, bone2Name=dyn_belt_l_05, swing1LimitDegrees=2.5, swing2LimitDegrees=6, twistLimitDegrees=10}
 {bone1Name=dyn_belt_l_05, bone2Name=dyn_belt_l_04, swing1LimitDegrees=2.5, swing2LimitDegrees=6, twistLimitDegrees=10}
 {bone1Name=dyn_belt_l_04, bone2Name=dyn_belt_l_03, swing1LimitDegrees=2.5, swing2LimitDegrees=6, twistLimitDegrees=10}
 {bone1Name=dyn_belt_l_08, bone2Name=dyn_belt_l_07, swing1LimitDegrees=2.5, swing2LimitDegrees=6, twistLimitDegrees=10}
 {bone1Name=dyn_belt_l_09, bone2Name=dyn_belt_l_08, swing1LimitDegrees=2.5, swing2LimitDegrees=6, twistLimitDegrees=10}
 {bone1Name=dyn_belt_l_010, bone2Name=dyn_belt_l_09, swing1LimitDegrees=2.5, swing2LimitDegrees=6, twistLimitDegrees=10}
 {bone1Name=dyn_belt_l_011, bone2Name=dyn_belt_l_010, swing1LimitDegrees=2.5, swing2LimitDegrees=6, twistLimitDegrees=10}
 {bone1Name=dyn_belt_r_08, bone2Name=dyn_belt_r_07, swing1LimitDegrees=5, swing2LimitDegrees=5, twistLimitDegrees=20}
 {bone1Name=dyn_belt_r_09, bone2Name=dyn_belt_r_08, swing1LimitDegrees=5, swing2LimitDegrees=5, twistLimitDegrees=20}
 {bone1Name=dyn_belt_r_010, bone2Name=dyn_belt_r_09, swing1LimitDegrees=5, swing2LimitDegrees=5, twistLimitDegrees=20}
 {bone1Name=dyn_belt_r_011, bone2Name=dyn_belt_r_010, swing1LimitDegrees=5, swing2LimitDegrees=5, twistLimitDegrees=20}
 {
  bone1Name=pelvis
  bone2Name=body_root
  swing1LimitDegrees=45
  swing1Motion=Free
  swing2LimitDegrees=45
  swing2Motion=Free
  twistLimitDegrees=45
  twistMotion=Free
 }
 {
  bone1Name=spine_01
  bone2Name=body_root
  swing1LimitDegrees=45
  swing1Motion=Free
  swing2LimitDegrees=45
  swing2Motion=Free
  twistLimitDegrees=45
  twistMotion=Free
 }
 {
  bone1Name=spine_02
  bone2Name=body_root
  swing1LimitDegrees=45
  swing1Motion=Free
  swing2LimitDegrees=45
  swing2Motion=Free
  twistLimitDegrees=45
  twistMotion=Free
 }
 {
  bone1Name=thigh_l
  bone2Name=body_root
  swing1LimitDegrees=45
  swing1Motion=Free
  swing2LimitDegrees=45
  swing2Motion=Free
  twistLimitDegrees=45
  twistMotion=Free
 }
 {
  bone1Name=calf_l
  bone2Name=body_root
  swing1LimitDegrees=45
  swing1Motion=Free
  swing2LimitDegrees=45
  swing2Motion=Free
  twistLimitDegrees=45
  twistMotion=Free
 }
 {
  bone1Name=thigh_r
  bone2Name=body_root
  swing1LimitDegrees=45
  swing1Motion=Free
  swing2LimitDegrees=45
  swing2Motion=Free
  twistLimitDegrees=45
  twistMotion=Free
 }
 {
  bone1Name=calf_r
  bone2Name=body_root
  swing1LimitDegrees=45
  swing1Motion=Free
  swing2LimitDegrees=45
  swing2Motion=Free
  twistLimitDegrees=45
  twistMotion=Free
 }
]
